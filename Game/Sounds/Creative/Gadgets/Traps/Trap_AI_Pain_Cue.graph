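# Sound cue for an AI pain voice on Creative trap gadgets, playing a CrateBot damage line.
# A weighted Random node, without replacement tracking, favors an empty branch, so the line plays on under a quarter of triggers.
# The result passes through a default Modulator, with reduced cue volume and pitch and Husk monster attenuation.
# A concurrency override allows one instance at a time and blocks new ones while it plays.
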SoundCue /Game/Sounds/Creative/Gadgets/Traps/Trap_AI_Pain_Cue
root=Modulator_1
nodes [
 {id=Modulator_1, children [Random_0]}
 {id=Random_0, children [WavePlayer_0,null], properties {bRandomizeWithoutReplacement=false, weights [0.3,1]}}
 {id=WavePlayer_0, properties {soundWaveAssetPtr=/Game/Sounds/Creative/VO/CrateBot/Cratebot_SeebotDamage_040}}
]
properties {
 attenuationSettings=/Game/Sounds/Attenuation/Fort_Monster_Husk_Att
 bOverrideConcurrency=true
 maxDistance=2800
 pitchMultiplier=0.9
 soundClassObject=&Fort_Gadgets_Creative
 volumeMultiplier=0.5
 concurrencyOverrides {maxCount=1, resolutionRule=PreventNew}
}
