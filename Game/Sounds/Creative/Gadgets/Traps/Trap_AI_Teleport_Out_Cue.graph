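# Sound cue for a Creative trap gadget, named for an AI teleport-out effect.
# A single teleporter placement wave plays through an Enveloper that fades volume to silence over a few seconds, with slight random volume variation.
# It uses the Creative gadgets sound class, small-toy focus attenuation and a reduced volume multiplier.
# Its own concurrency override rejects new instances once the limit is reached.
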SoundCue /Game/Sounds/Creative/Gadgets/Traps/Trap_AI_Teleport_Out_Cue
root=Enveloper_0
nodes [
 {
  id=Enveloper_0
  children [WavePlayer_4]
  properties {volumeMax=1.05, volumeMin=0.95, pitchCurve {editorCurveData {keys [{time=0}]}}, volumeCurve {editorCurveData {keys [<at 0 1>,<at 2.5 0>]}}}
 }
 {id=WavePlayer_4, properties {soundWaveAssetPtr=/Game/Sounds/Fort_Gadgets/Teleporter/teleporter_place_02}}
]
properties {
 attenuationSettings=/Game/Sounds/Creative/Attenuations/Fort_Creative_SmallToy_Focus_Att
 bOverrideConcurrency=true
 maxDistance=2000
 soundClassObject=&Fort_Gadgets_Creative
 volumeMultiplier=0.6
 concurrencyOverrides {resolutionRule=PreventNew}
}
