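# Sound cue for equipping the Creative BiPlane trap gadget, routed to the non-critical in-game UI sound class.
# A Mixer layers a ballistic turret equip sound, at reduced level, with a Ferret vehicle engine-start sound.
# An Enveloper at the root fades the blend out over one second and adds slight random pitch and volume variation.
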SoundCue /Game/Sounds/Creative/Gadgets/Traps/Trap_BiPlane_Equip_Cue
root=Enveloper_0
nodes [
 {
  id=Enveloper_0
  children [Mixer_0]
  properties {pitchMin=0.95, volumeMin=0.95, pitchCurve {editorCurveData {keys [{time=0}]}}, volumeCurve {editorCurveData {keys [<at 0 1>,<at 1 0>]}}}
 }
 {id=Mixer_0, children [WavePlayer_1, WavePlayer_11], properties {inputVolume [0.6,1]}}
 {id=WavePlayer_1, properties {soundWaveAssetPtr=/Game/Sounds/Fort_Traps/MountedTurret/Turret_Ballistic_Equip_01}}
 {id=WavePlayer_11, properties {soundWaveAssetPtr=/Game/Athena/Sounds/Vehicles/Ferret/Ferret_Start_01}}
]
properties {maxDistance=2097152, soundClassObject=&Fort_UI_InGame_NonCritical, volumeMultiplier=0.48}
