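# Sound cue for equipping the electric trap gadget in Creative, played through the non-critical in-game UI sound class.
# It reuses the electric wall trap's firing wave, pitched up and lowered in volume at the cue level.
# An Enveloper fades that wave from full volume to silence in under a second and adds slight random pitch and volume variation.
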SoundCue /Game/Sounds/Creative/Gadgets/Traps/Trap_Electric_Equip_Cue
root=Enveloper_0
nodes [
 {
  id=Enveloper_0
  children [WavePlayer_2]
  properties {
   pitchMax=1.05
   pitchMin=0.95
   volumeMin=0.95
   pitchCurve {editorCurveData {keys [{time=0}]}}
   volumeCurve {editorCurveData {keys [<at 0 1>,<at 0.75 0>]}}
  }
 }
 {id=WavePlayer_2, properties {soundWaveAssetPtr=/Game/Sounds/Fort_Traps/Wall_Electric/WallTrap_Electric_Fire_001}}
]
properties {
 maxDistance=2097152
 pitchMultiplier=1.45
 soundClassObject=&Fort_UI_InGame_NonCritical
 volumeMultiplier=0.48
}
