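# Sound cue for a Creative-mode trap gadget, apparently played when an elimination trap becomes active.
# Mixes three reused waves, a power-down, a jump pad destroy and an electric trap placement (behind a Delay node), each pitched up and faded out by its own Enveloper.
# A root Modulator wraps the mix, which plays in the Creative gadgets sound class with overridden distance-filtered, focus-aware attenuation and prevent-new concurrency.
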SoundCue /Game/Sounds/Creative/Gadgets/Traps/Trap_Elimination_Active_Cue
root=Modulator_15
nodes [
 {id=Delay_3, children [Enveloper_1]}
 {
  id=Enveloper_1
  children [WavePlayer_7]
  properties {pitchMax=1.2, pitchMin=1.2, pitchCurve {editorCurveData {keys [{time=0}]}}, volumeCurve {editorCurveData {keys [<at 0 1>,<at 2.2 0>]}}}
 }
 {
  id=Enveloper_2
  children [WavePlayer_13]
  properties {
   pitchMax=1.5
   pitchMin=1.5
   pitchCurve {editorCurveData {keys [{time=0}]}}
   volumeCurve {editorCurveData {keys [{interpMode=RCIM_Cubic, tangentMode=RCTM_Break, time=0, value=1}, <at 1.75 0>]}}
  }
 }
 {
  id=Enveloper_3
  children [WavePlayer_12]
  properties {pitchMax=1.6, pitchMin=1.6, pitchCurve {editorCurveData {keys [{time=0}]}}, volumeCurve {editorCurveData {keys [<at 0 1>,<at 1.5 0>]}}}
 }
 {id=Mixer_0, children [Enveloper_3,Enveloper_2,Delay_3], properties {inputVolume [0.6,1,1]}}
 {id=Modulator_15, children [Mixer_0]}
 {id=WavePlayer_12, properties {soundWaveAssetPtr=/Game/Sounds/Fort_Atlas_Sounds/atlas_new_powerdown}}
 {id=WavePlayer_13, properties {soundWaveAssetPtr=/Game/Sounds/Fort_Traps/JumpPad/JumpPad_Destroy_003}}
 {id=WavePlayer_7, properties {soundWaveAssetPtr=/Game/Sounds/Fort_Traps/Electric_Trap/Electric_Trap_Place01}}
]
properties {
 attenuationSettings=&Fort_Traps_Att
 bOverrideAttenuation=true
 bOverrideConcurrency=true
 maxDistance=2000
 soundClassObject=&Fort_Gadgets_Creative
 volumeMultiplier=0.4
 attenuationOverrides {
  bAttenuateWithLPF=true
  bEnableListenerFocus=true
  falloffDistance=1200
  hPFFrequencyAtMax=2000
  lPFFrequencyAtMax=4000
  lPFRadiusMax=4000
  lPFRadiusMin=2000
  nonFocusVolumeAttenuation=0.3
  reverbDistanceMax=5000
  reverbDistanceMin=3000
  reverbWetLevelMax=5
  stereoSpread=512
  attenuationShapeExtents {x=800}
 }
 concurrencyOverrides {resolutionRule=PreventNew}
}
