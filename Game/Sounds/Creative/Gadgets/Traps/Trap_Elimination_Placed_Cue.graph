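# Sound cue played when a Creative elimination trap is placed, built from reused Fort trap sounds.
# A mixer layers three branches at stepped volumes: a modulated random pick of two Wall Lights placement waves, a modulated Enemy Launcher placement wave, and an unmodulated Lights fire wave.
# It uses the traps attenuation and placement sound class, and overrides concurrency with a prevent-new rule.
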SoundCue /Game/Sounds/Creative/Gadgets/Traps/Trap_Elimination_Placed_Cue
root=Mixer_12
nodes [
 {id=Mixer_12, children [Modulator_11,Modulator_16,WavePlayer_0], properties {inputVolume [1,0.8,0.6]}}
 {id=Modulator_11, children [Random_23]}
 {id=Modulator_16, children [WavePlayer_139]}
 {id=Random_23, children [WavePlayer_80, WavePlayer_81], properties {weights [1,1]}}
 {id=WavePlayer_0, properties {soundWaveAssetPtr=/Game/Sounds/Fort_Traps/Wall_Lights/Lights_Fire_02b}}
 {id=WavePlayer_139, properties {soundWaveAssetPtr=/Game/Sounds/Fort_Traps/Enemy_Launcher/EnemyLauncher_Placed_001}}
 {id=WavePlayer_80, properties {soundWaveAssetPtr=/Game/Sounds/Fort_Traps/Wall_Lights/WallTrap_Lights_Placed_002}}
 {id=WavePlayer_81, properties {soundWaveAssetPtr=/Game/Sounds/Fort_Traps/Wall_Lights/WallTrap_Lights_Placed_001}}
]
properties {
 attenuationSettings=&Fort_Traps_Att
 bOverrideConcurrency=true
 maxDistance=6000
 soundClassObject=/Game/Sounds/FortSoundClassesAndModes/Fort_Traps_Placement
 volumeMultiplier=0.55
 concurrencyOverrides {resolutionRule=PreventNew}
}
