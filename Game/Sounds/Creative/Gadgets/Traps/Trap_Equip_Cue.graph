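# Sound cue for equipping a trap, stored with the Creative gadget trap sounds.
# A single wave player feeds the root directly, playing the mounted turret ballistic equip sound borrowed from the Fort trap audio, with no randomization or modulation.
# It routes to the non-critical in-game UI sound class at roughly half volume.
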SoundCue /Game/Sounds/Creative/Gadgets/Traps/Trap_Equip_Cue
root=WavePlayer_1
nodes [{id=WavePlayer_1, properties {soundWaveAssetPtr=/Game/Sounds/Fort_Traps/MountedTurret/Turret_Ballistic_Equip_01}}]
properties {maxDistance=2097152, soundClassObject=&Fort_UI_InGame_NonCritical, volumeMultiplier=0.48}
