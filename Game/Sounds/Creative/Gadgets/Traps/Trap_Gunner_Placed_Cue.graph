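# Sound cue for placing the Gunner trap gadget in Creative, playing one of four mechanical set sounds.
# A short fixed delay feeds a modulator with pitch variation flattened, then a random node picks among the equally weighted waves.
# It uses shared trap attenuation and the trap placement sound class at reduced volume, with a concurrency override that blocks new instances once the limit is reached.
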SoundCue /Game/Sounds/Creative/Gadgets/Traps/Trap_Gunner_Placed_Cue
root=Delay_0
nodes [
 {id=Delay_0, children [Modulator_15], properties {delayMax=0.02, delayMin=0.02}}
 {id=Modulator_15, children [Random_0], properties {pitchMax=1, pitchMin=1}}
 {id=Random_0, children [WavePlayer_0,WavePlayer_1,WavePlayer_2,WavePlayer_3], properties {weights [1,1,1,1]}}
 {id=WavePlayer_0, properties {soundWaveAssetPtr=&/Mech_Gadget_GunnerTrap_Set_04}}
 {id=WavePlayer_1, properties {soundWaveAssetPtr=&/Mech_Gadget_GunnerTrap_Set_01}}
 {id=WavePlayer_2, properties {soundWaveAssetPtr=&/Mech_Gadget_GunnerTrap_Set_02}}
 {id=WavePlayer_3, properties {soundWaveAssetPtr=&/Mech_Gadget_GunnerTrap_Set_03}}
]
properties {
 attenuationSettings=&Fort_Traps_Att
 bOverrideConcurrency=true
 maxDistance=6000
 soundClassObject=/Game/Sounds/FortSoundClassesAndModes/Fort_Traps_Placement
 volumeMultiplier=0.6
 concurrencyOverrides {resolutionRule=PreventNew}
}
