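# Sound cue for equipping the Hamster trap gadget in Creative, routed to the non-critical in-game UI sound class.
# A mixer layers a ballistic turret equip wave, at reduced level, with a random pick from three equally weighted soft Hamster vehicle impacts.
# An enveloper at the root fades the mix out along a volume curve and adds slight random pitch and volume variation.
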SoundCue /Game/Sounds/Creative/Gadgets/Traps/Trap_Hamster_Equip_Cue
root=Enveloper_0
nodes [
 {
  id=Enveloper_0
  children [Mixer_0]
  properties {pitchMin=0.95, volumeMin=0.95, pitchCurve {editorCurveData {keys [{time=0}]}}, volumeCurve {editorCurveData {keys [<at 0 1>,<at 1 0>]}}}
 }
 {id=Mixer_0, children [WavePlayer_1,Random_0], properties {inputVolume [0.7,1]}}
 {id=Random_0, children [WavePlayer_0,WavePlayer_2,WavePlayer_4], properties {weights [1,1,1]}}
 {id=WavePlayer_0, properties {soundWaveAssetPtr=/Game/Athena/Sounds/Vehicles/Hamster/Hamster_Impact_Soft_03}}
 {id=WavePlayer_1, properties {soundWaveAssetPtr=/Game/Sounds/Fort_Traps/MountedTurret/Turret_Ballistic_Equip_01}}
 {id=WavePlayer_2, properties {soundWaveAssetPtr=/Game/Athena/Sounds/Vehicles/Hamster/Hamster_Impact_Soft_01}}
 {id=WavePlayer_4, properties {soundWaveAssetPtr=/Game/Athena/Sounds/Vehicles/Hamster/Hamster_Impact_Soft_02}}
]
properties {maxDistance=2097152, soundClassObject=&Fort_UI_InGame_NonCritical, volumeMultiplier=0.48}
