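# Sound cue for placing the Creative ice block trap, using trap attenuation, the trap placement sound class and a small concurrency cap.
# A quality-level switch at the root sends its first two slots to a shared mix of a construction-done sound and a modulated freeze trap placement sound.
# Its last two slots each play a single "opt" placement wave through their own modulator with slight pitch and volume variation.
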SoundCue /Game/Sounds/Creative/Gadgets/Traps/Trap_IceBlock_Placed_cue.Trap_IceBlock_Placed_Cue
root=QualityLevel_2
nodes [
 {id=Mixer_2, children [WavePlayer_15, Modulator_6], properties {inputVolume [0.5,1]}}
 {id=Modulator_6, children [WavePlayer_14], properties {pitchMax=1.2, volumeMax=1.5, volumeMin=1}}
 {id=Modulator_7, children [WavePlayer_18], properties {pitchMax=1.1, pitchMin=0.9, volumeMax=1.1}}
 {id=Modulator_8, children [WavePlayer_19], properties {pitchMax=1.1, pitchMin=0.9, volumeMax=1.1}}
 {id=QualityLevel_2, children [Mixer_2,Mixer_2,Modulator_7,Modulator_8]}
 {id=WavePlayer_14, properties {soundWaveAssetPtr=/Game/Sounds/Fort_Traps/Freeze_Trap/Freeze_Trap_Place_01}}
 {id=WavePlayer_15, properties {soundWaveAssetPtr=/Game/Sounds/Fort_UI_Sounds/Gameplay/Fort_Contructino_Done_02}}
 {id=WavePlayer_18, properties {soundWaveAssetPtr=/Game/Sounds/Fort_Traps/Freeze_Trap/Freeze_Trap_Place_opt}}
 {id=WavePlayer_19, properties {soundWaveAssetPtr=&Freeze_Trap_Place_opt}}
]
properties {
 attenuationSettings=&Fort_Traps_Att
 bOverrideConcurrency=true
 maxDistance=6000
 soundClassObject=/Game/Sounds/FortSoundClassesAndModes/Fort_Traps_Placement
 volumeMultiplier=0.45
 concurrencyOverrides {maxCount=3}
}
