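# Looping hum sound cue for the Creative item spawner trap's hologram, by its name, in the Creative gadgets sound class.
# A looping wave player plays a reused Machine_Hum test sound through a Modulator that fixes the pitch slightly raised, at reduced overall volume.
# The cue overrides attenuation for short-range falloff and uses interpolated listener focus, which attenuates the hum when it is out of focus.
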SoundCue /Game/Sounds/Creative/Gadgets/Traps/Trap_ItemSpawner_Holo_Loop_Cue
root=Modulator_0
nodes [
 {id=Modulator_0, children [WavePlayer_3], properties {pitchMax=1.2, pitchMin=1.2}}
 {id=WavePlayer_3, properties {bLooping=true, soundWaveAssetPtr=/Game/Sounds/Fort_Test_Sounds/World_Objects/Mesh_Effects/Machine_Hum}}
]
properties {
 bOverrideAttenuation=true
 maxDistance=500
 soundClassObject=&Fort_Gadgets_Creative
 volumeMultiplier=0.3
 attenuationOverrides {
  bEnableFocusInterpolation=true
  bEnableListenerFocus=true
  falloffDistance=400
  focusAzimuth=15
  nonFocusAzimuth=45
  nonFocusVolumeAttenuation=0.5
  attenuationShapeExtents {x=100}
 }
}
