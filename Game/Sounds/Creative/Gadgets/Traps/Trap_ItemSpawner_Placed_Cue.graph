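# Sound cue that plays when an item spawner trap is placed in Creative.
# It reuses a dart trap placement sound layered with a quieter construction-done UI sound through a mixer.
# A fixed pitch-up modulator and a very short fixed delay wrap the mix.
# It uses trap attenuation and the trap placement sound class, with overridden concurrency that prevents new instances at the limit.
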SoundCue /Game/Sounds/Creative/Gadgets/Traps/Trap_ItemSpawner_Placed_Cue
root=Delay_0
nodes [
 {id=Delay_0, children [Modulator_15], properties {delayMax=0.02, delayMin=0.02}}
 {id=Mixer_15, children [WavePlayer_40, WavePlayer_104], properties {inputVolume [1,0.4]}}
 {id=Modulator_15, children [Mixer_15], properties {pitchMax=1.2, pitchMin=1.2}}
 {id=WavePlayer_104, properties {soundWaveAssetPtr=/Game/Sounds/Fort_UI_Sounds/Gameplay/Fort_Contructino_Done_02}}
 {id=WavePlayer_40, properties {soundWaveAssetPtr=/Game/Sounds/Fort_Traps/Dart_Trap/WallTrap_Darts_Placed_001}}
]
properties {
 attenuationSettings=&Fort_Traps_Att
 bOverrideConcurrency=true
 maxDistance=6000
 soundClassObject=/Game/Sounds/FortSoundClassesAndModes/Fort_Traps_Placement
 volumeMultiplier=0.6
 concurrencyOverrides {resolutionRule=PreventNew}
}
