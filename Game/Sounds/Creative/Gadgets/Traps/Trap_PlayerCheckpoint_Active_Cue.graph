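# Sound cue for the Creative player checkpoint trap, named as its active sound and routed to the trap placement sound class with trap attenuation.
# A mixer layers one of three random Ward Trap beam pulses with a GhostRocks activate sound pitched up by a fixed modulator.
# The mix passes through a pitch-neutral modulator and a brief fixed delay.
# A concurrency override prevents new instances once the limit is reached.
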SoundCue /Game/Sounds/Creative/Gadgets/Traps/Trap_PlayerCheckpoint_Active_Cue
root=Delay_0
nodes [
 {id=Delay_0, children [Modulator_15], properties {delayMax=0.02, delayMin=0.02}}
 {id=Mixer_0, children [Random_0,Modulator_0], properties {inputVolume [1,1]}}
 {id=Modulator_0, children [WavePlayer_0], properties {pitchMax=1.5, pitchMin=1.5, volumeMax=1, volumeMin=1}}
 {id=Modulator_15, children [Mixer_0], properties {pitchMax=1, pitchMin=1}}
 {id=Random_0, children [WavePlayer_4,WavePlayer_5,WavePlayer_6], properties {weights [1,1,1]}}
 {id=WavePlayer_0, properties {soundWaveAssetPtr=/Game/Sounds/Fort_GamePlay_Sounds/Player/GhostRocks/GhostRocks_Activate_01}}
 {id=WavePlayer_4, properties {soundWaveAssetPtr=/Game/Sounds/Fort_Traps/Ward_Trap/Ward_Trap_Beam_Pulse06}}
 {id=WavePlayer_5, properties {soundWaveAssetPtr=/Game/Sounds/Fort_Traps/Ward_Trap/Ward_Trap_Beam_Pulse05}}
 {id=WavePlayer_6, properties {soundWaveAssetPtr=/Game/Sounds/Fort_Traps/Ward_Trap/Ward_Trap_Beam_Pulse04}}
]
properties {
 attenuationSettings=&Fort_Traps_Att
 bOverrideConcurrency=true
 maxDistance=6000
 soundClassObject=/Game/Sounds/FortSoundClassesAndModes/Fort_Traps_Placement
 volumeMultiplier=0.6
 concurrencyOverrides {resolutionRule=PreventNew}
}
