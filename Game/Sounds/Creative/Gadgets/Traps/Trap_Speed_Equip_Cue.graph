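# Sound cue for equipping the Creative speed trap gadget, routed to the non-critical in-game UI sound class.
# A Mixer layers the launch pad UI select sound with a quieter, randomly pitched-up speed trap fire sound.
# An Enveloper at the root applies a rising pitch curve and a short volume fade-out to the blend, with slight random pitch and volume variation.
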SoundCue /Game/Sounds/Creative/Gadgets/Traps/Trap_Speed_Equip_Cue
root=Enveloper_0
nodes [
 {
  id=Enveloper_0
  children [Mixer_0]
  properties {
   pitchMax=1.05
   pitchMin=0.95
   volumeMin=0.95
   pitchCurve {editorCurveData {keys [{time=0}, <at 1 0.7>]}}
   volumeCurve {editorCurveData {keys [<at 0 1>,<at 0.65 0>]}}
  }
 }
 {id=Mixer_0, children [Modulator_0,WavePlayer_0], properties {inputVolume [0.75,1]}}
 {id=Modulator_0, children [WavePlayer_7], properties {pitchMax=1.2, pitchMin=1.1}}
 {id=WavePlayer_0, properties {soundWaveAssetPtr=/Game/Athena/Sounds/LaunchPad/Launch_Pad_UI_Select}}
 {id=WavePlayer_7, properties {soundWaveAssetPtr=/Game/Sounds/Fort_Traps/Speed_Trap/Speed_Curve_Trap_Fire01}}
]
properties {maxDistance=2097152, soundClassObject=&Fort_UI_InGame_NonCritical, volumeMultiplier=0.45}
