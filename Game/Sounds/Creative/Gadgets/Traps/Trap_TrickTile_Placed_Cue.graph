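# Sound cue for placing the Trick Tile trap in Creative, reusing Gunner Trap "Set" mechanical recordings.
# A short fixed delay feeds a modulator, which plays one of three equally weighted random wave variants.
# The cue is pitched up and quieted, uses trap attenuation and the trap placement sound class, and overrides concurrency to cap simultaneous instances by stopping the oldest.
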SoundCue /Game/Sounds/Creative/Gadgets/Traps/Trap_TrickTile_Placed_Cue
root=Delay_0
nodes [
 {id=Delay_0, children [Modulator_15], properties {delayMax=0.02, delayMin=0.02}}
 {id=Modulator_15, children [Random_0]}
 {id=Random_0, children [WavePlayer_4,WavePlayer_6,WavePlayer_7], properties {weights [1,1,1]}}
 {id=WavePlayer_4, properties {soundWaveAssetPtr=&/Mech_Gadget_GunnerTrap_Set_04}}
 {id=WavePlayer_6, properties {soundWaveAssetPtr=&/Mech_Gadget_GunnerTrap_Set_02}}
 {id=WavePlayer_7, properties {soundWaveAssetPtr=&/Mech_Gadget_GunnerTrap_Set_03}}
]
properties {
 attenuationSettings=&Fort_Traps_Att
 bOverrideConcurrency=true
 maxDistance=6000
 pitchMultiplier=1.2
 soundClassObject=/Game/Sounds/FortSoundClassesAndModes/Fort_Traps_Placement
 volumeMultiplier=0.6
 concurrencyOverrides {maxCount=3, resolutionRule=StopOldest}
}
