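# Sound cue for the Creative trap waypoint dot, playing a single one-shot wave.
# One wave player feeds the output directly, reusing a gingerbread pistol gumdrop sound from the weapon audio.
# The cue routes to the trap sound class at slightly reduced volume and overrides attenuation with its own falloff, shape extent and stereo spread for positional playback.
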SoundCue /Game/Sounds/Creative/Gadgets/Traps/Trap_Waypoint_Dot_Cue
root=WavePlayer_0
nodes [{id=WavePlayer_0, properties {soundWaveAssetPtr=/Game/Sounds/Fort_Weapons/pistol/Pistol_Gingerbread_Gumpdrop_PU_01}}]
properties {
 bOverrideAttenuation=true
 maxDistance=1800
 soundClassObject=&Fort_Traps
 volumeMultiplier=0.8
 attenuationOverrides {falloffDistance=1000, stereoSpread=600, attenuationShapeExtents {x=800}}
}
