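# Sound cue for the ring sound of a Creative trap waypoint gadget.
# A single wave player feeds the Trap_Waypoint_Ring_02 wave straight to the output, with no randomization or modulation.
# The cue uses the Fort_Traps sound class at slightly reduced volume and overrides attenuation inline, setting its own falloff distance, shape extent and stereo spread.
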SoundCue /Game/Sounds/Creative/Gadgets/Traps/Trap_Waypoint_Ring_Cue
root=WavePlayer_1
nodes [{id=WavePlayer_1, properties {soundWaveAssetPtr=&/Trap_Waypoint_Ring_02}}]
properties {
 bOverrideAttenuation=true
 maxDistance=1800
 soundClassObject=&Fort_Traps
 volumeMultiplier=0.9
 attenuationOverrides {falloffDistance=1000, stereoSpread=600, attenuationShapeExtents {x=800}}
}
